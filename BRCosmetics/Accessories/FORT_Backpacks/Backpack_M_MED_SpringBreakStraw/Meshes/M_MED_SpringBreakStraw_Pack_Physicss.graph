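# Physics asset for the SpringBreakStraw backpack cosmetic, simulating a dangling charm on its left side.
# A large root capsule represents the pack, with a charm base capsule locked rigidly to it.
# Small capsules form a chain from the base through five dynamic bones, then branch into three tip bones off the last link.
# Angular limits allow a wide primary swing, a narrower secondary swing and almost no twist, with tighter swing on the branches.
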PhysicsAsset /BRCosmetics/Accessories/FORT_Backpacks/Backpack_M_MED_SpringBreakStraw/Meshes/M_MED_SpringBreakStraw_Pack_Physicss
bodies {
 base_charm_l [{shapeType=Capsule, length=0.505, radius=0.505, shapeName=base_charm_l_capsule, rotation=<rot 0 -90 0>}]
 dyn_charm_00_l [{
   shapeType=Capsule
   length=0.6221
   radius=0.5
   shapeName=dyn_charm_00_l_capsule
   center=<1.2426 -0.0006 0.1772>
   rotation=<rot 87.9446 -0.1744 -0.2206>
  }]
 dyn_charm_01_l [{
   shapeType=Capsule
   length=0.6474
   radius=0.5
   shapeName=dyn_charm_01_l_capsule
   center=<0.7752 0 0.1925>
   rotation=<rot 88.1679 -179.9994 -179.9995>
  }]
 dyn_charm_02_l [{
   shapeType=Capsule
   length=0.3967
   radius=0.7118
   shapeName=dyn_charm_02_l_capsule
   center=<0.669 0.0128 0>
   rotation=<rot 88.3809 177.1648 -179.9997>
  }]
 dyn_charm_03_l [{
   shapeType=Capsule
   length=0.5854
   radius=0.5
   shapeName=dyn_charm_03_l_capsule
   center=<0.8586 0 0>
   rotation=<rot 88.927 -179.9961 178.2173>
  }]
 dyn_charm_04_l [{shapeType=Capsule, length=0.0598, radius=1.2071, shapeName=dyn_charm_04_l_capsule, center=<1.6884 0 0>, rotation=<rot 90 0 0>}]
 dyn_charm_05_l [{shapeType=Capsule, length=1.7724, radius=0.5, shapeName=dyn_charm_05_l_capsule, center=<1.304 0.2332 -0.0027>, rotation=<rot -90 0 0>}]
 dyn_charm_06_l [{shapeType=Capsule, length=1.7724, radius=0.5, shapeName=dyn_charm_06_l_capsule, center=<1.4342 -0.0244 0.0076>, rotation=<rot -90 0 0>}]
 dyn_charm_07_l [{shapeType=Capsule, length=1.7724, radius=0.5, shapeName=dyn_charm_07_l_capsule, center=<1.2802 -0.2258 -0.0505>, rotation=<rot -90 0 0>}]
 root [{shapeType=Capsule, length=2.7025, radius=7.625, shapeName=root_capsule, center=<0.0859 -8.551 2.228>, rotation=<rot -0.0002 -90 6.8909>}]
}
constraints [
 {bone1Name=dyn_charm_01_l, bone2Name=dyn_charm_00_l, swing1LimitDegrees=40, swing2LimitDegrees=10, twistLimitDegrees=1}
 {bone1Name=dyn_charm_02_l, bone2Name=dyn_charm_01_l, swing1LimitDegrees=40, swing2LimitDegrees=10, twistLimitDegrees=1}
 {bone1Name=dyn_charm_03_l, bone2Name=dyn_charm_02_l, swing1LimitDegrees=40, swing2LimitDegrees=10, twistLimitDegrees=1}
 {bone1Name=dyn_charm_04_l, bone2Name=dyn_charm_03_l, swing1LimitDegrees=40, swing2LimitDegrees=15, twistLimitDegrees=1}
 {bone1Name=dyn_charm_00_l, bone2Name=base_charm_l, swing1LimitDegrees=40, swing2LimitDegrees=10, twistLimitDegrees=1}
 {bone1Name=base_charm_l, bone2Name=root, swing1LimitDegrees=0, swing2LimitDegrees=0, twistLimitDegrees=0}
 {bone1Name=dyn_charm_05_l, bone2Name=dyn_charm_04_l, swing1LimitDegrees=20, swing2LimitDegrees=10, twistLimitDegrees=1}
 {bone1Name=dyn_charm_06_l, bone2Name=dyn_charm_04_l, swing1LimitDegrees=15, swing2LimitDegrees=7, twistLimitDegrees=1}
 {bone1Name=dyn_charm_07_l, bone2Name=dyn_charm_04_l, swing1LimitDegrees=20, swing2LimitDegrees=10, twistLimitDegrees=1}
]
